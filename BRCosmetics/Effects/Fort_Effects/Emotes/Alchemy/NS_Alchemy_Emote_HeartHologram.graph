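# Niagara system for the Alchemy emote's blue heart hologram, which pops up beside the emote's origin, flickers on, holds, then fades out after about three seconds.
# Two single-burst mesh emitters layer a main heart and a slightly larger, subtler one, with age-driven scale and alpha curves.
# A Light emitter adds a similarly flickering light and airglow sprite.
# User.bIsFrontendPreview switches the meshes' size, opacity and offset, but leaves the light unchanged.
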NiagaraSystem /BRCosmetics/Effects/Fort_Effects/Emotes/Alchemy/NS_Alchemy_Emote_HeartHologram
dependencies {usedDataInterfaces [&NiagaraDataInterfaceCurve], usedDynamicInputs [], usedModules [&SystemState]}
emitters {
 HeartMesh {
  emitterUpdateScript [
   &EmitterState(
    Inactive Response=::NewEnumerator1
    Life Cycle Mode=::NewEnumerator1
    Loop Behavior=::NewEnumerator1
    Loop Duration Mode=::NewEnumerator1
    UseLoopDelay=false
   )
   &SpawnBurst_Instantaneous(Spawn Count=1, Use Spawn Probability=false)
  ]
  particleSpawnScript [
   &InitializeParticle(
    Color=&MakeCustomLinearColorFromBool(Bool=$User.bIsFrontendPreview, True Color=<rgba 1.3901 1.8503 20 10>, False Color=<rgba 1.3901 1.8503 20 1>)
    Mesh Scale Mode=::NewEnumerator3
    Mesh Uniform Scale=&MakeCustomFloatFromBool(Bool=$User.bIsFrontendPreview, True Float=1.2, False Float=3)
    UsePositionOffset=false
    Write Material Random=true
   )
   &InitialMeshOrientation(Orientation Vector=<0 1 0>, Rotation=<0 0 0.075>, Use Orientation Vector=true, Use Rotation Vector=true)
   &SystemLocation(MaskSpawnGroup=false, Offset=<-22 15 140>)
  ]
  particleUpdateScript [
   &ParticleState(Kill Particles When Lifetime Has Elapsed=false, Let Infinitely Lived Particles Die When Emitter Deactivates=true)
   &ScaleMeshSize(Scale Factor=&VectorFromFloat(Value=&FloatFromCurve(
      FloatCurve=DataInterface() {propertyValues {Curve {keys [
          {arriveTangent=0.0732, interpMode=RCIM_Cubic, leaveTangent=0.0732, tangentMode=RCTM_User, time=0, value=0}
          {arriveTangent=-2.4957, interpMode=RCIM_Cubic, leaveTangent=-2.4957, tangentMode=RCTM_User, time=0.4, value=1}
         ]}}}
      CurveIndex=$Emitter.Age
     )))
   &SystemLocation(
    MaskSpawnGroup=false
    Offset=&Lerp_Vector(
     A=Linked() {linkedVariable {name=Particles.Initial.Position, type=NiagaraPosition}}
     B=&Add_Vector(
      A=Linked() {linkedVariable {name=Particles.Initial.Position, type=NiagaraPosition}}
      B=&MakeCustomVectorFromBool(Bool=$User.bIsFrontendPreview, True Vector=<-25 50 0>, False Vector=<-50 100 0>)
     )
     Alpha=&FloatFromCurve(FloatCurve=DataInterface() {propertyValues {Curve {keys [<at 0 1>,<at 0.4 1>]}}}, CurveIndex=$Emitter.Age)
    )
   ) {module=SystemLocation002}
   &ScaleColor(
    Scale Alpha=&FloatFromCurve(
     FloatCurve=DataInterface() {propertyValues {Curve {keys [<at 0 0>,<at 0.1 0.5>,<at 0.2 0.1>,<at 0.3 0.8>,<at 0.4 0.1>,<at 0.8 1>,<at 3.1 1>,<at 3.4 0>]}}}
     CurveIndex=$Emitter.Age
    )
    ScaleA=true
    ScaleRGB=true
   )
  ]
  rendererClasses [&NiagaraMeshRendererProperties]
  renderers [{rendererClass=&NiagaraMeshRendererProperties, rendererIndex=0}]
 }
 HeartMesh_Subtle {
  emitterUpdateScript [
   &EmitterState(
    Inactive Response=::NewEnumerator1
    Life Cycle Mode=::NewEnumerator1
    Loop Behavior=::NewEnumerator1
    Loop Duration Mode=::NewEnumerator1
    UseLoopDelay=false
   )
   &SpawnBurst_Instantaneous(Spawn Count=1, Use Spawn Probability=false)
  ]
  particleSpawnScript [
   &InitializeParticle(
    Color=&MakeCustomLinearColorFromBool(Bool=$User.bIsFrontendPreview, True Color=<rgba 1.3901 1.8503 20 1>, False Color=<rgba 1.3901 1.8503 20 0.025>)
    Mesh Scale Mode=::NewEnumerator3
    Mesh Uniform Scale=&MakeCustomFloatFromBool(Bool=$User.bIsFrontendPreview, True Float=1.3, False Float=3.2)
    UsePositionOffset=false
    Write Material Random=true
   )
   &InitialMeshOrientation(Orientation Vector=<0 1 0>, Rotation=<0 0 0.075>, Use Orientation Vector=true, Use Rotation Vector=true)
   &SystemLocation(MaskSpawnGroup=false, Offset=<-22 15 140>)
  ]
  particleUpdateScript [
   &ParticleState(Kill Particles When Lifetime Has Elapsed=false, Let Infinitely Lived Particles Die When Emitter Deactivates=true)
   &ScaleMeshSize(Scale Factor=&VectorFromFloat(Value=&FloatFromCurve(
      FloatCurve=DataInterface() {propertyValues {Curve {keys [
          {arriveTangent=0.0732, interpMode=RCIM_Cubic, leaveTangent=0.0732, tangentMode=RCTM_User, time=0, value=0}
          {arriveTangent=-2.4956, interpMode=RCIM_Cubic, leaveTangent=-2.4956, tangentMode=RCTM_User, time=0.4, value=1}
         ]}}}
      CurveIndex=$Emitter.Age
     )))
   &SystemLocation(
    MaskSpawnGroup=false
    Offset=&Lerp_Vector(
     A=Linked() {linkedVariable {name=Particles.Initial.Position, type=NiagaraPosition}}
     B=&Add_Vector(
      A=Linked() {linkedVariable {name=Particles.Initial.Position, type=NiagaraPosition}}
      B=&MakeCustomVectorFromBool(Bool=$User.bIsFrontendPreview, True Vector=<-25 50 0>, False Vector=<-50 100 0>)
     )
     Alpha=&FloatFromCurve(FloatCurve=DataInterface() {propertyValues {Curve {keys [<at 0 1>,<at 0.4 1>]}}}, CurveIndex=$Emitter.Age)
    )
   ) {module=SystemLocation002}
   &ScaleColor(
    Scale Alpha=&FloatFromCurve(
     FloatCurve=DataInterface() {propertyValues {Curve {keys [<at 0 0>,<at 0.1 0.8>,<at 0.2 0.1>,<at 0.3 0.8>,<at 0.4 0.1>,<at 0.8 1>,<at 3.1 1>,<at 3.4 0>]}}}
     CurveIndex=$Emitter.Age
    )
    ScaleA=true
    ScaleRGB=true
   )
  ]
  rendererClasses [&NiagaraMeshRendererProperties]
  renderers [{rendererClass=&NiagaraMeshRendererProperties, rendererIndex=0}]
 }
 Light {
  emitterUpdateScript [
   &EmitterState(Inactive Response=::NewEnumerator1, Life Cycle Mode=::NewEnumerator1, Loop Duration=1, UseLoopDelay=false)
   &SpawnBurst_Instantaneous(Spawn Count=1, Use Spawn Probability=false)
  ]
  particleSpawnScript [
   &InitializeParticle(
    Color=<rgba 0.0695 0.0925 1 1>
    Sprite Size Mode=::NewEnumerator3
    Uniform Sprite Size=200
    UsePositionOffset=false
    Write Material Random=true
   )
   &SystemLocation(MaskSpawnGroup=false, Offset=<-22 15 140>)
  ]
  particleUpdateScript [
   &ParticleState()
   /SetVariables(Particles.Airglow_Color:LinearColor=&MultiplyLinearColor(
     Linear Color=<rgba 0.0695 0.0925 1 0.1>
     Scale Factor=&MakeVector4(
      X=1
      Y=1
      Z=1
      W=&FloatFromCurve(
       FloatCurve=DataInterface() {propertyValues {Curve {keys [<at 0 0>,<at 0.1 0.8>,<at 0.2 0.1>,<at 0.3 0.8>,<at 0.4 0.1>,<at 0.8 1>,<at 3 1>,<at 3.3 0>]}}}
       CurveIndex=$Emitter.Age
      )
     )
    )) {isSetParameters=true}
   &SystemLocation(
    MaskSpawnGroup=false
    Offset=&Lerp_Vector(
     A=Linked() {linkedVariable {name=Particles.Initial.Position, type=NiagaraPosition}}
     B=&Add_Vector(A=Linked() {linkedVariable {name=Particles.Initial.Position, type=NiagaraPosition}}, B=<-50 100 0>)
     Alpha=&FloatFromCurve(FloatCurve=DataInterface() {propertyValues {Curve {keys [<at 0 1>,<at 0.4 1>]}}}, CurveIndex=$Emitter.Age)
    )
   ) {module=SystemLocation002}
   &FN_Light_Attributes(Create Exponent Attribute=true, Create Radius Attribute=true, Exponent=3, Radius=400)
   &ScaleColor(
    Scale RGB=&VectorFromFloat(Value=&FloatFromCurve(
      FloatCurve=DataInterface() {propertyValues {Curve {keys [<at 0 0>,<at 0.1 0.8>,<at 0.2 0.1>,<at 0.3 0.8>,<at 0.4 0.1>,<at 0.8 1>,<at 3.1 1>,<at 3.4 0>]}}}
      CurveIndex=$Emitter.Age
     ))
    ScaleA=true
    ScaleRGB=true
   )
  ]
  rendererClasses [&NiagaraLightRendererProperties, &NiagaraSpriteRendererProperties]
  renderers [
   {rendererClass=&NiagaraLightRendererProperties, rendererIndex=0}
   {rendererClass=&NiagaraSpriteRendererProperties, rendererIndex=1}
  ]
 }
}
summary {
 systemName=NS_Alchemy_Emote_HeartHologram
 emitters [
  {emitterName=HeartMesh, rendererClasses [&NiagaraMeshRendererProperties]}
  {emitterName=HeartMesh_Subtle, rendererClasses [&NiagaraMeshRendererProperties]}
  {emitterName=Light, rendererClasses [&NiagaraLightRendererProperties, &NiagaraSpriteRendererProperties]}
 ]
 userVariables [User.bIsFrontendPreview {type=NiagaraBool}]
}
